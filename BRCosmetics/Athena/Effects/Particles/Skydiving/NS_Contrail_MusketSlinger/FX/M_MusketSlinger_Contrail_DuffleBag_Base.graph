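# Base lit material for Niagara mesh-particle duffle bags in the MusketSlinger skydiving contrail, giving them a procedural wobble.
# Character Parent 2023 functions build the surface attributes from texture inputs.
# Two sine waves along local Y and Z, offset by time and particle random, sum into a push along the normalized local position.
# A dynamic parameter scales the Y wave's gradient, and a capsule distance field masks the displacement near the bag's core.
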Material /BRCosmetics/Athena/Effects/Particles/Skydiving/NS_Contrail_MusketSlinger/FX/M_MusketSlinger_Contrail_DuffleBag_Base
graph {
 properties {
  bUsedWithNiagaraMeshParticles=true
  pixelDepthOffsetMode=PDOM_Legacy
  previewMesh=&/SM_MusketSlinger_MoneyBag_v2
  usedShadingModels=Default Lit
 }
 nodes [
  Constant()
  n1 = Constant() {r=1}
  n2 = PreSkinnedPosition()
  n3 = VertexInterpolator(VS=@n2)
  n4 = MaterialFunctionCall!(UV) {materialFunction=/Game/Global/Materials/Character_Parent_2023/Functions/MF_CharacterParent_2023_TextureInputs}
  GetMaterialAttributes(None=MaterialFunctionCall!(
    Tech Art Mask UVs
    Emissive UVs
    Fresnel=Fresnel(ExponentIn=@n1) {baseReflectFraction=0}
    LocalPosition=ShaderStageSwitch(PixelShader=@n3.PS, VertexShader=@n2)
    Mask Input=@n4.M
    Material=@n4.M
   ) {materialFunction=/Game/Global/Materials/Character_Parent_2023/Functions/MF_CharacterParent_2023_BasicTechArt}.Result) {attributeGetTypes [
    #1
    #2
    #3
    #4
    #5
   ]}
  n5 = Multiply(A=Add(A=Time(), B=ParticleRandom()), B=ScalarParameter() {defaultValue=2.5, parameterName=Speed})
  n6 = MaterialFunctionCall() {materialFunction=&LocalPosition}
  n7 = ComponentMask(None=@n6.Instance Local Position) {g=true}
  n8 = ScalarParameter() {defaultValue=36, parameterName=Bounds Size}
  n9 = Multiply(
   A=ConstantBiasScale(None=Divide(A=@n7, B=@n8)) {desc=Gradient for wave}
   B=Multiply(
    A=DynamicParameter() {parameterIndex=1, paramNames [WiggleHorizontalDirection, Param2, Param3, Param4]}.WiggleHorizontalDirection
    B=ScalarParameter() {defaultValue=1.3, parameterName=Frequency}
   )
  )
  n10 = Multiply(A=Sine(None=Add(A=@n5, B=@n9)), B=ScalarParameter() {defaultValue=4, parameterName=Amplitude})
  n11 = ComponentMask(None=@n6.Instance Local Position) {b=true}
  n12 = Constant() {r=40}
  n13 = Multiply(A=ConstantBiasScale(None=Divide(A=@n11, B=@n12)), B=Constant() {r=1})
  n14 = Multiply(A=Add(A=Time(), B=ParticleRandom()), B=Constant() {r=1.3})
  n15 = Multiply(A=Sine(None=Add(A=@n13, B=@n14)), B=Constant() {r=3})
  LinearInterpolate(A=@n10, B=@n15) {constAlpha=0}
  n16 = ScalarParameter() {defaultValue=12, parameterName=Capsule Length}
  n17 = MaterialFunctionCall!(
   Coordinates=@n6.Instance Local Position
   P0=MaterialFunctionCall(G=Multiply(A=@n16) {constB=-1}) {materialFunction=&AppendMany}.RGB
   P1=MaterialFunctionCall(G=@n16) {materialFunction=&AppendMany}.RGB
   Width=ScalarParameter() {defaultValue=18, parameterName=Capsule Mask Width}
  ) {materialFunction=/Engine/Functions/DistanceFields/Shapes/DistanceField_Capsule}
  n18 = ScalarParameter() {defaultValue=11, parameterName=Capsule Mask Falloff Distance}
  n19 = Normalize(VectorInput=@n6.Instance Local Position)
  Multiply(
   A=Saturate(None=Divide(A=@n17.Result, B=@n18))
   B=Multiply(A=Transform(None=@n19) {desc=The vector to push towards, transformSourceType=TRANSFORMSOURCE_Instance}, B=Add(A=@n10, B=@n15))
  )
 ]
}
