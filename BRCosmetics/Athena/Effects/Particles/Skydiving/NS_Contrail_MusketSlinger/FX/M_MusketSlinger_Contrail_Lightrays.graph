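# Unlit, two-sided translucent Niagara mesh-particle material for the light rays in the MusketSlinger skydiving contrail.
# Two opposing panned ray textures and noise combine with a center gradient, blended toward a Ceil-stepped version for stylization, then shaped by noise-varied V-gradient masks.
# Particle color tints the result, while dynamic parameters drive random offset, pan speed and U tiling.
# Opacity adds near-camera fade, inverse-Fresnel edge softening and depth fade, which cheaper platforms can skip.
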Material /BRCosmetics/Athena/Effects/Particles/Skydiving/NS_Contrail_MusketSlinger/FX/M_MusketSlinger_Contrail_Lightrays
graph {
 properties {
  bCastRayTracedShadows=false
  bUsedWithNiagaraMeshParticles=true
  blendMode=BLEND_Translucent
  pixelDepthOffsetMode=PDOM_Legacy
  shadingModel=MSM_Unlit
  twoSided=true
  usedShadingModels=Unlit
 }
 nodes [
  n1 = ParticleColor()
  n2 = ComponentMask(None=TextureCoordinate()) {g=true}
  n3 = DynamicParameter() {defaultValue=<rgba 0 1 1 1>, paramNames [Random 0-1, Panner Speed, "U Tile ", Param4]}
  n4 = Multiply(A=TextureCoordinate(), B=@n3."U Tile ")
  n5 = ScalarParameter() {defaultValue=4, parameterName=Gradient Tile A}
  n6 = Multiply(A=@n3.Panner Speed, B=Add(A=@n3.Random 0-1, B=Time()))
  n7 = TextureSample(UVs=Panner(Coordinate=Multiply(A=@n4, B=@n5), Time=@n6) {speedX=0.0702}) {texture=/Game/Effects/Fort_Effects/Textures/Outlander/T_InTheZone_LightRays_01}
  n8 = Multiply(A=@n3."U Tile ", B=TextureCoordinate())
  n9 = ScalarParameter() {defaultValue=3, parameterName=Gradient Tile B}
  n10 = TextureSample(UVs=Panner(Coordinate=Multiply(A=@n8, B=@n9), Time=@n6) {speedX=-0.047}) {texture=&T_InTheZone_LightRays_01}
  n11 = TextureCoordinate() {uTiling=0.33, vTiling=0.01}
  n12 = DynamicParameter() {defaultValue=<rgba 0 1 1 1>, paramNames [Random 0-1, Panner Speed, "U Tile ", Param4]}
  n13 = TextureSample(UVs=Panner(Coordinate=Add(A=@n11, B=@n12.Random 0-1), Time=@n6) {speedX=0.0436}) {texture=&T_NoisePack_01}
  n14 = ScalarParameter() {defaultValue=3, parameterName=Mask Brightness}
  n15 = TextureCoordinate()
  n16 = ScalarParameter() {defaultValue=0.02, parameterName=Slide}
  n17 = OneMinus(None=ComponentMask(None=Add(A=@n15, B=@n16)) {g=true})
  n18 = ScalarParameter() {parameterName=Center Exponent}
  n19 = Add(
   A=Multiply(A=Multiply(A=@n7.R, B=@n10.G), B=Multiply(A=@n13.R, B=@n14))
   B=Multiply(A=Power(Base=@n17, Exp=@n18), B=ScalarParameter() {defaultValue=1, parameterName=Center Brightness})
  )
  n20 = ScalarParameter() {defaultValue=1, parameterName=Ceil Bias}
  n21 = Multiply(A=Ceil(None=Add(A=@n19, B=@n20)), B=ScalarParameter() {defaultValue=3, parameterName=Ceil Brightness})
  n22 = ScalarParameter() {defaultValue=0.5, parameterName=Style Lerp}
  n23 = OneMinus(None=ComponentMask(None=TextureCoordinate()) {g=true})
  n24 = TextureCoordinate() {uTiling=3, vTiling=0}
  n25 = LinearInterpolate(Alpha=TextureSample(UVs=Panner(Coordinate=@n24) {speedX=0.2, speedY=0.1}) {texture=&T_NoisePack_01}.R) {constA=2, constB=6}
  n26 = Multiply(A=Saturate(None=Multiply(A=@n2) {constB=3}), B=Multiply(A=LinearInterpolate(A=@n19, B=@n21, Alpha=@n22), B=Power(Base=@n23, Exp=@n25)))
  MaterialFunctionCall(Emissive=Multiply(A=@n1.RGB, B=@n26)) {materialFunction=&MF_FX_EmissiveAdjustments}
  n27 = ScalarParameter() {defaultValue=60, parameterName=NearCamFade_Distance}
  n28 = Multiply(A=@n1.A, B=@n26)
  n29 = DepthFade(FadeDistance=ScalarParameter() {parameterName=Depth Fade Distance})
  n30 = Constant() {r=1}
  n31 = StaticSwitchParameter(
   True=FeatureLevelSwitch(Default=@n29, ES2=@n30, ES3_1=@n30, SM4_REMOVED=@n29)
   False=QualitySwitch(Default=@n29, Low=Constant() {r=1}, High=@n29, Medium=@n29)
  ) {parameterName=Use_MobilePerf_Settings}
  Multiply(
   A=Multiply(
    A=MaterialFunctionCall(Distance=@n27) {materialFunction=&Near_Camera_Fade}.Result
    B=Multiply(A=@n28, B=@n31)
   )
   B=OneMinus(None=Fresnel() {baseReflectFraction=0, exponent=0.1})
  )
 ]
}
