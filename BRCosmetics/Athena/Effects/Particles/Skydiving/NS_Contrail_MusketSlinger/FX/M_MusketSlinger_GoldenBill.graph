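# Two-sided lit material for gold-tinted bill mesh particles in the MusketSlinger skydiving contrail.
# Vertex color blends the band/edge texture with the bill texture's red channel tinted by ColorBaseBill.
# A time-driven panner over UVs, offset by particle random, is shaped with Frac, Floor and Power into a periodic gold glint.
# The glint is maxed with a dim base tint, masked by vertex color and fed through MF_FX_EmissiveAdjustments.
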Material /BRCosmetics/Athena/Effects/Particles/Skydiving/NS_Contrail_MusketSlinger/FX/M_MusketSlinger_GoldenBill
graph {
 properties {
  bUsedWithNiagaraMeshParticles=true
  pixelDepthOffsetMode=PDOM_Legacy
  previewMesh=&/SM_MusketSlinger_PileOBills
  twoSided=true
  usedShadingModels=Default Lit
 }
 nodes [
  Constant() {r=0.9}
  Constant() {r=1}
  Constant() {r=0.8}
  n1 = TextureSampleParameter2D() {
   parameterName=BandColor
   samplerType=SAMPLERTYPE_LinearColor
   texture=&/T_MusketSlinger_BillStack_Edges
  }
  n2 = TextureSampleParameter2D() {parameterName=TextureBill, texture=&/T_MusketSlinger_Bill_D}
  n3 = VectorParameter() {parameterName=ColorBaseBill, defaultValue=<rgba 0.9 0.5 0.252 0>}
  n4 = VertexColor()
  LinearInterpolate(A=@n1.RGB, B=Multiply(A=@n2.R, B=@n3.RGB), Alpha=@n4)
  LinearInterpolate(Alpha=@n4) {constA=0.9, constB=0.7}
  n5 = ComponentMask(None=TextureCoordinate()) {r=true}
  n6 = Add(A=Multiply(A=Add(B=@n5)) {constB=0.12}) {constB=0}
  n7 = Add(A=ParticleRandom(), B=Time())
  n8 = Constant() {r=1.6}
  n9 = Frac(None=ComponentMask(None=Panner(Coordinate=@n6, Time=@n7, Speed=@n8) {speedX=0.2, speedY=0.2}) {g=true})
  n10 = Add(A=@n9) {constB=0.1}
  n11 = OneMinus(None=Multiply(A=Subtract(A=@n9) {constB=0.9}) {constB=10})
  n12 = Multiply(A=Constant3Vector() {constant=<rgba 1.2 0.705 0.3 1>}, B=Multiply(A=Floor(None=@n10), B=Power(Base=@n11) {constExponent=3}))
  n13 = Multiply(A=@n3.RGB) {constB=0.02}
  MaterialFunctionCall(Emissive=LinearInterpolate(B=Max(A=@n12, B=@n13), Alpha=VertexColor())) {materialFunction=&MF_FX_EmissiveAdjustments}
 ]
}
