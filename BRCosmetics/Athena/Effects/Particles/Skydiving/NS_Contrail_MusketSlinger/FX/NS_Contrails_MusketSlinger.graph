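# Niagara skydiving contrail for the MusketSlinger cosmetic: glowing money bags trail the diver, shedding bills, stacks, coins and confetti.
# MoneyBags particles drive those child emitters through attribute-reader spawning and sampling, then emit location events near 70% of their life.
# BillsExplosion and CoinsExplosion define outward bursts, but their listed spawn modules are disabled.
# Dive amount scales rates, motion and GPU speed lines, with separate front-end tuning.
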NiagaraSystem /BRCosmetics/Athena/Effects/Particles/Skydiving/NS_Contrail_MusketSlinger/FX/NS_Contrails_MusketSlinger
dependencies {
 usedDataInterfaces [
  &NiagaraDataInterfaceCurve
  &NiagaraDataInterfaceVector2DCurve
  &NiagaraDataInterfaceParticleRead
  &NiagaraDataInterfaceMeshRendererInfo
  &NiagaraDataInterfaceArrayFloat
  &NiagaraDataInterfaceCamera
  &NiagaraDataInterfaceColorCurve
  &NiagaraDataInterfaceSpriteRendererInfo
 ]
 usedDynamicInputs []
 usedModules [
  &SystemState
  /Game/Effects/Niagara/Modules/Skydiving/FN_DivingAttributes/v5/FN_DivingAttributes
  &self:SystemSpawnScript.SystemScriptSource.SystemScriptGraph.NiagaraNodeAssignment_0.SetVariables_#1
  &FN_Integrator
 ]
}
emitters {
 Bills {
  emitterUpdateScript [
   &EmitterState()
   &SpawnRate(SpawnRate=100, Use Spawn Probability=false) {enabled=false}
   &SpawnParticlesFromOtherEmitter(
    Attribute Reader=DataInterface() {propertyValues {EmitterBinding {bindingMode=Other, emitterName=MoneyBags}}}
    Maximum Age Spawn Threshold=2
    Spawn Rate=&FN_FrontEndFloatModifier(Gameplay=15, Front End=10)
    Use Maximum Age=false
    Use Minimum Age=false
   )
  ]
  particleSpawnScript [
   &InitializeParticle(
    Color=<rgba 1 0.5467 0.15 1>
    Lifetime=&FN_FrontEndFloatModifier(
     Gameplay=&RandomRangeFloat(Minimum=0.7, Maximum=1.1)
     Front End=&RandomRangeFloat(Minimum=0.5, Maximum=1)
    )
    Mass Mode=::NewEnumerator1
    Mesh Renderer Array Visibility Mode=::NewEnumerator3
    Mesh Renderer Info=DataInterface() {propertyValues {MeshRenderer=&self:Bills_0.NiagaraMeshRendererProperties_0}}
    Mesh Scale Mode=::NewEnumerator3
    UsePositionOffset=false
    Write Material Random=true
   )
   &SampleParticlesFromOtherEmitter(
    Applied Position Offset:Vector3f=<0 0 20>
    Lifetime Sampling:ENiagara_DirectReadSamplingMode=::NewEnumerator1
    Maximum Sampled Particle Age=2
    Module.MaxSampledAgeCutoff=true
    Module.MinSampledAgeCutoff=false
    Reject Particle By Sampled Age=true
    Velocity Sampling=::NewEnumerator0
   )
   &AddVelocity(Velocity=&RandomRangeVector(Minimum=<0 0 100>, Maximum=<0 0 130>)) {enabled=false}
   &ShapeLocation(MaskSpawnGroup=false, Override Seed=false, Sphere Radius=8, UseOwnerScale=true)
   &AddVelocity(
    Module.Use Falloff Distance=false
    Origin Offset=<0 0 -3>
    Override Seed=false
    Velocity Mode=::NewEnumerator1
    Velocity Origin=$Output.SampleParticlesFromOtherEmitter.SampledPosition
    Velocity Speed=&RandomRangeFloat(Minimum=120, Maximum=160)
   ) {module=AddVelocity001}
   &InitialMeshOrientation(Use Rotation Vector=false)
  ]
  particleUpdateScript [
   &ParticleState()
   &UpdateMeshOrientation(Rotation Rate=&RandomRangeFloat(Minimum=-1, Maximum=1), Rotation Vector=<1 1 1>)
   &ScaleMeshSize(Scale Factor=&VectorFromFloat(Value=&FloatFromCurve(FloatCurve=DataInterface() {propertyValues {Curve {keys [
          {interpMode=RCIM_Cubic, tangentMode=RCTM_User, time=0.05, value=0}
          {interpMode=RCIM_Cubic, tangentMode=RCTM_User, time=0.3, value=1}
          {interpMode=RCIM_Cubic, tangentMode=RCTM_User, time=0.7, value=1}
          {interpMode=RCIM_Cubic, tangentMode=RCTM_User, time=1, value=0}
         ]}}})))
   &CurlNoiseForce(
    Noise Frequency=&Multiply_Float(
     A=&RandomRangeFloat(Minimum=250, Maximum=500)
     B=&Lerp_Float(A=1, B=4, Alpha=$System.Dive)
    )
    Noise Strength=&Multiply_Float(
     A=&RandomRangeFloat(Minimum=800, Maximum=1000)
     B=&FloatFromCurve(FloatCurve=DataInterface() {propertyValues {Curve {keys [{interpMode=RCIM_Cubic, tangentMode=RCTM_User, time=0, value=0}, {interpMode=RCIM_Cubic, tangentMode=RCTM_User, time=0.4, value=1}]}}})
    )
    Pan Noise=false
   )
   &AerodynamicDrag(
    Aerodynamic Drag=0.25
    Pivot Offset Maximum:Vector3f=<1 0.5 -0.5>
    Pivot Offset Minimum:Vector3f=<-1 -0.5 -2>
    Use Linear Drag=true
    Use Rotational Drag=true
   ) {enabled=false}
   &Drag(Drag=&RandomRangeFloat(Minimum=0.2, Maximum=0.3), Use Linear Drag=true, Use Rotational Drag=true) {enabled=false}
   &SolveForcesAndVelocity(Clamp Velocity=false, Limit Acceleration=false)
   &AlignSpriteToMeshOrientation(SetAlignment=true, SetFacing=true)
  ]
  rendererClasses [&NiagaraMeshRendererProperties]
  renderers [{rendererClass=&NiagaraMeshRendererProperties, rendererIndex=0}]
 }
 BillsExplosion {
  emitterUpdateScript [
   &EmitterState()
   &SpawnBurst_Instantaneous(Spawn Count=100, Use Spawn Probability=false, UseLoopCountLimit=false) {enabled=false}
   &SpawnParticlesFromOtherEmitter(
    Attribute Reader=DataInterface() {propertyValues {EmitterBinding {bindingMode=Other, emitterName=MoneyBags}}}
    Spawn Rate=&Lerp_Float(A=10, B=18, Alpha=$System.Dive)
    Use Maximum Age=false
    Use Minimum Age=false
   ) {enabled=false}
  ]
  particleSpawnScript [
   &InitializeParticle(
    Lifetime Max=1
    Lifetime Min=0.5
    Lifetime Mode=::NewEnumerator1
    Mesh Renderer Array Visibility Mode=::NewEnumerator3
    Mesh Renderer Info=DataInterface() {propertyValues {MeshRenderer=&self:BillsExplosion_1.NiagaraMeshRendererProperties_4}}
    Mesh Scale Mode=::NewEnumerator3
    UsePositionOffset=false
    Write Material Random=true
   )
   &InheritVelocity(Inherited Velocity Amount Scale=<0 0 -1>, Speed Limit=true, UseSpeedThreshold=false) {enabled=false}
   &ShapeLocation(MaskSpawnGroup=false, Override Seed=false, Sphere Radius=20, UseOwnerScale=true)
   &AddVelocity(
    Module.Use Falloff Distance=false
    Origin Offset=<0 0 -2>
    Override Seed=false
    Velocity Mode=::NewEnumerator1
    Velocity Speed=&RandomRangeFloat(Minimum=250, Maximum=300)
   )
   &InitialMeshOrientation(Use Rotation Vector=false)
  ]
  particleUpdateScript [
   &ParticleState()
   &UpdateMeshOrientation(Rotation Rate=0.8, Rotation Vector=<1 1 1>)
   &ScaleMeshSize(Scale Factor=&VectorFromFloat(Value=&FloatFromCurve(FloatCurve=DataInterface() {propertyValues {Curve {keys [
          {interpMode=RCIM_Cubic, tangentMode=RCTM_User, time=0, value=0}
          {interpMode=RCIM_Cubic, tangentMode=RCTM_User, time=0.2, value=1}
          {interpMode=RCIM_Cubic, tangentMode=RCTM_User, time=0.7, value=1}
          {interpMode=RCIM_Cubic, tangentMode=RCTM_User, time=1, value=0}
         ]}}})))
   &CurlNoiseForce(
    Noise Frequency=&Multiply_Float(
     A=&RandomRangeFloat(Minimum=30, Maximum=50)
     B=&Lerp_Float(A=1, B=4, Alpha=$System.Dive)
    )
    Noise Strength=&RandomRangeFloat(Minimum=500, Maximum=700)
    Pan Noise=false
   ) {enabled=false}
   &Drag(Use Linear Drag=true, Use Rotational Drag=true)
   &GravityForce(Gravity=<0 0 -100>) {enabled=false}
   &SolveForcesAndVelocity(Clamp Velocity=false, Limit Acceleration=false)
  ]
  rendererClasses [&NiagaraMeshRendererProperties]
  renderers [{rendererClass=&NiagaraMeshRendererProperties, rendererIndex=0}]
 }
 Bills_Stacks {
  emitterUpdateScript [
   &EmitterState()
   &SpawnParticlesFromOtherEmitter(
    Attribute Reader=DataInterface() {propertyValues {EmitterBinding {bindingMode=Other, emitterName=MoneyBags}}}
    Maximum Age Spawn Threshold=2
    Spawn Rate=&FN_FrontEndFloatModifier(Gameplay=10, Front End=8)
    Use Maximum Age=false
    Use Minimum Age=false
   )
  ]
  particleSpawnScript [
   &InitializeParticle(
    Color=<rgba 1 0.5467 0.15 1>
    Lifetime=1.3
    Mesh Renderer Array Visibility Mode=::NewEnumerator3
    Mesh Renderer Info=DataInterface() {propertyValues {MeshRenderer=&self:Bills_Stacks_1.NiagaraMeshRendererProperties_0}}
    Mesh Scale Mode=::NewEnumerator3
    Mesh Uniform Scale=2
    UsePositionOffset=false
    Write Material Random=true
   )
   &SampleParticlesFromOtherEmitter(
    Applied Position Offset:Vector3f=<0 0 20>
    Lifetime Sampling:ENiagara_DirectReadSamplingMode=::NewEnumerator1
    Maximum Sampled Particle Age=2
    Module.MaxSampledAgeCutoff=true
    Module.MinSampledAgeCutoff=false
    Reject Particle By Sampled Age=true
    Velocity Sampling=::NewEnumerator0
   )
   &InheritVelocity(Inherited Velocity Amount Scale=<0 0 -1>, Speed Limit=true, UseSpeedThreshold=false) {enabled=false}
   &ShapeLocation(MaskSpawnGroup=false, Offset=<0 0 5>, Offset Mode=::NewEnumerator0, Override Seed=false, Sphere Radius=20, UseOwnerScale=true)
   &AddVelocity(
    Module.Use Falloff Distance=false
    Origin Offset=<0 0 -0.8>
    Override Seed=false
    Velocity Mode=::NewEnumerator1
    Velocity Origin=$Output.SampleParticlesFromOtherEmitter.SampledPosition
    Velocity Speed=&RandomRangeFloat(Minimum=25, Maximum=40)
   ) {module=AddVelocity001}
   &InitialMeshOrientation(Use Rotation Vector=false)
  ]
  particleUpdateScript [
   &ParticleState()
   &UpdateMeshOrientation(Rotation Rate=0.4, Rotation Vector=<1 1 1>)
   &ScaleMeshSize(Scale Factor=&VectorFromFloat(Value=&FloatFromCurve(FloatCurve=DataInterface() {propertyValues {Curve {keys [
          {interpMode=RCIM_Cubic, tangentMode=RCTM_User, time=0.05, value=0}
          {interpMode=RCIM_Cubic, tangentMode=RCTM_User, time=0.3, value=1}
          {interpMode=RCIM_Cubic, tangentMode=RCTM_User, time=0.8, value=1}
          {interpMode=RCIM_Cubic, tangentMode=RCTM_User, time=1, value=0}
         ]}}})))
   &CurlNoiseForce(
    Noise Frequency=&Multiply_Float(
     A=&RandomRangeFloat(Minimum=5, Maximum=10)
     B=&Lerp_Float(A=1, B=4, Alpha=$System.Dive)
    )
    Noise Strength=&RandomRangeFloat(Minimum=15, Maximum=35)
    Pan Noise=false
   )
   &Drag(Drag=0.2, Use Linear Drag=true, Use Rotational Drag=true) {enabled=false}
   &SolveForcesAndVelocity(Clamp Velocity=false, Limit Acceleration=false)
  ]
  rendererClasses [&NiagaraMeshRendererProperties]
  renderers [{rendererClass=&NiagaraMeshRendererProperties, rendererIndex=0}]
 }
 Coins {
  emitterUpdateScript [
   &EmitterState()
   &SpawnParticlesFromOtherEmitter(
    Attribute Reader=DataInterface() {propertyValues {EmitterBinding {bindingMode=Other, emitterName=MoneyBags}}}
    Maximum Age Spawn Threshold=2
    Spawn Rate=&FN_FrontEndFloatModifier(Gameplay=15, Front End=10)
    Use Maximum Age=false
    Use Minimum Age=false
   )
  ]
  particleSpawnScript [
   &InitializeParticle(Lifetime Max=1.3, Lifetime Mode=::NewEnumerator1, Mesh Scale Mode=::NewEnumerator3, UsePositionOffset=false, Write Material Random=true)
   &SampleParticlesFromOtherEmitter(
    Applied Position Offset:Vector3f=<0 0 20>
    Maximum Sampled Particle Age=2
    Module.MaxSampledAgeCutoff=true
    Module.MinSampledAgeCutoff=false
    Reject Particle By Sampled Age=true
    Velocity Sampling=::NewEnumerator0
   )
   &InheritVelocity(Inherited Velocity Amount Scale=<0 0 -1>, Speed Limit=true, UseSpeedThreshold=false) {enabled=false}
   &ShapeLocation(MaskSpawnGroup=false, Override Seed=false, Sphere Radius=14, UseOwnerScale=true)
   &AddVelocity(
    Module.Use Falloff Distance=false
    Origin Offset=<0 0 -0.7>
    Override Seed=false
    Velocity Mode=::NewEnumerator1
    Velocity Origin=$Output.SampleParticlesFromOtherEmitter.SampledPosition
    Velocity Speed=&RandomRangeFloat(Minimum=20, Maximum=35)
   )
   &InitialMeshOrientation(Use Rotation Vector=false)
  ]
  particleUpdateScript [
   &ParticleState()
   &UpdateMeshOrientation(Rotation Rate=0.8, Rotation Vector=<1 1 1>)
   &ScaleMeshSize(Scale Factor=&VectorFromFloat(Value=&FloatFromCurve(FloatCurve=DataInterface() {propertyValues {Curve {keys [
          {interpMode=RCIM_Cubic, tangentMode=RCTM_User, time=0, value=0}
          {interpMode=RCIM_Cubic, tangentMode=RCTM_User, time=0.2, value=1}
          {interpMode=RCIM_Cubic, tangentMode=RCTM_User, time=0.7, value=1}
          {interpMode=RCIM_Cubic, tangentMode=RCTM_User, time=1, value=0}
         ]}}})))
   &CurlNoiseForce(
    Noise Frequency=&Multiply_Float(
     A=&RandomRangeFloat(Minimum=30, Maximum=50)
     B=&Lerp_Float(A=1, B=4, Alpha=$System.Dive)
    )
    Noise Strength=&RandomRangeFloat(Minimum=500, Maximum=700)
    Pan Noise=false
   ) {enabled=false}
   &Drag(Drag=0.1, Use Linear Drag=true, Use Rotational Drag=true) {enabled=false}
   &SolveForcesAndVelocity(Clamp Velocity=false, Limit Acceleration=false)
  ]
  rendererClasses [&NiagaraMeshRendererProperties]
  renderers [{rendererClass=&NiagaraMeshRendererProperties, rendererIndex=0}]
 }
 CoinsExplosion {
  emitterUpdateScript [
   &EmitterState()
   &SpawnBurst_Instantaneous(Spawn Count=100, Use Spawn Probability=false, UseLoopCountLimit=false) {enabled=false}
   &SpawnParticlesFromOtherEmitter(
    Attribute Reader=DataInterface() {propertyValues {EmitterBinding {bindingMode=Other, emitterName=MoneyBags}}}
    Spawn Rate=&Lerp_Float(A=10, B=18, Alpha=$System.Dive)
    Use Maximum Age=false
    Use Minimum Age=false
   ) {enabled=false}
  ]
  particleSpawnScript [
   &InitializeParticle(
    Lifetime Max=1
    Lifetime Min=0.5
    Lifetime Mode=::NewEnumerator1
    Mesh Scale Mode=::NewEnumerator3
    UsePositionOffset=false
    Write Material Random=true
   )
   &InheritVelocity(Inherited Velocity Amount Scale=<0 0 -1>, Speed Limit=true, UseSpeedThreshold=false) {enabled=false}
   &ShapeLocation(MaskSpawnGroup=false, Override Seed=false, Sphere Radius=20, UseOwnerScale=true)
   &AddVelocity(
    Module.Use Falloff Distance=false
    Origin Offset=<0 0 -0.5>
    Override Seed=false
    Velocity Mode=::NewEnumerator1
    Velocity Speed=&RandomRangeFloat(Minimum=250, Maximum=300)
   )
   &InitialMeshOrientation(Use Rotation Vector=false)
  ]
  particleUpdateScript [
   &ParticleState()
   &UpdateMeshOrientation(Rotation Rate=0.8, Rotation Vector=<1 1 1>)
   &ScaleMeshSize(Scale Factor=&VectorFromFloat(Value=&FloatFromCurve(FloatCurve=DataInterface() {propertyValues {Curve {keys [
          {interpMode=RCIM_Cubic, tangentMode=RCTM_User, time=0, value=0}
          {interpMode=RCIM_Cubic, tangentMode=RCTM_User, time=0.2, value=1}
          {interpMode=RCIM_Cubic, tangentMode=RCTM_User, time=0.7, value=1}
          {interpMode=RCIM_Cubic, tangentMode=RCTM_User, time=1, value=0}
         ]}}})))
   &CurlNoiseForce(
    Noise Frequency=&Multiply_Float(
     A=&RandomRangeFloat(Minimum=30, Maximum=50)
     B=&Lerp_Float(A=1, B=4, Alpha=$System.Dive)
    )
    Noise Strength=&RandomRangeFloat(Minimum=500, Maximum=700)
    Pan Noise=false
   ) {enabled=false}
   &Drag(Use Linear Drag=true, Use Rotational Drag=true)
   &GravityForce(Gravity=<0 0 -100>)
   &SolveForcesAndVelocity(Clamp Velocity=false, Limit Acceleration=false)
  ]
  rendererClasses [&NiagaraMeshRendererProperties]
  renderers [{rendererClass=&NiagaraMeshRendererProperties, rendererIndex=0}]
 }
 Confetti {
  emitterUpdateScript [
   &EmitterState()
   &SpawnParticlesFromOtherEmitter(
    Attribute Reader=DataInterface() {propertyValues {EmitterBinding {bindingMode=Other, emitterName=MoneyBags}}}
    Spawn Rate=&Lerp_Float(A=20, B=30, Alpha=$System.Dive)
    Use Maximum Age=false
    Use Minimum Age=false
   )
  ]
  particleSpawnScript [
   &InitializeParticle(
    Color Maximum=<rgba 6 5.16 4.2 1>
    Color Minimum=<rgba 3 2.58 2.4 1>
    Color Mode=::NewEnumerator2
    Lifetime=0.8
    Sprite Rotation Mode=::NewEnumerator2
    Sprite Size Mode=::NewEnumerator1
    Uniform Sprite Size Max=1.5
    Uniform Sprite Size Min=0.5
    UsePositionOffset=false
    Write Material Random=true
   )
   &SampleParticlesFromOtherEmitter(
    Applied Position Offset:Vector3f=<0 0 20>
    Maximum Sampled Particle Age=2
    Module.MaxSampledAgeCutoff=true
    Module.MinSampledAgeCutoff=false
    Reject Particle By Sampled Age=true
    Velocity Sampling=::NewEnumerator0
   )
   &InheritVelocity(Inherited Velocity Amount Scale=<0 0 -1>, Speed Limit=true, UseSpeedThreshold=false) {enabled=false}
   &ShapeLocation(MaskSpawnGroup=false, Override Seed=false, Sphere Radius=16, UseOwnerScale=true)
   &AddVelocity(
    Module.Use Falloff Distance=false
    Origin Offset=<0 0 -3>
    Override Seed=false
    Velocity Mode=::NewEnumerator1
    Velocity Origin=$Output.SampleParticlesFromOtherEmitter.SampledPosition
    Velocity Speed=&RandomRangeFloat(Minimum=20, Maximum=50)
   )
   &InitialMeshOrientation(Use Rotation Vector=false) {enabled=false}
  ]
  particleUpdateScript [
   &ParticleState()
   &ScaleSpriteSize(Uniform Curve Sprite Scale=DataInterface() {propertyValues {Curve {keys [{interpMode=RCIM_Cubic, tangentMode=RCTM_User, time=0.3, value=1}, {interpMode=RCIM_Cubic, tangentMode=RCTM_User, time=1, value=0}]}}})
   &UpdateMeshOrientation(Rotation Rate=0.5, Rotation Vector=<1 1 1>) {enabled=false}
   &CurlNoiseForce(
    Noise Frequency=&Multiply_Float(
     A=&RandomRangeFloat(Minimum=30, Maximum=50)
     B=&Lerp_Float(A=1, B=4, Alpha=$System.Dive)
    )
    Noise Strength=&RandomRangeFloat(Minimum=500, Maximum=700)
    Pan Noise=false
   ) {enabled=false}
   &Drag(Drag=0.4, Use Linear Drag=true, Use Rotational Drag=true) {enabled=false}
   &SolveForcesAndVelocity(Clamp Velocity=false, Limit Acceleration=false)
   &SubUVAnimation(
    "Sprite Renderer "=DataInterface() {propertyValues {SpriteRenderer=&self:Confetti_1.NiagaraSpriteRendererProperties_0}}
    UseEndFrame=false
    UseStartFrame=false
   ) {enabled=false}
   &ScaleColor(
    Scale Alpha=&FloatFromCurve(FloatCurve=DataInterface() {propertyValues {Curve {keys [
         {interpMode=RCIM_Cubic, tangentMode=RCTM_User, time=0, value=0}
         {interpMode=RCIM_Cubic, tangentMode=RCTM_User, time=0.2, value=1}
         {interpMode=RCIM_Cubic, tangentMode=RCTM_User, time=0.7, value=1}
         {interpMode=RCIM_Cubic, tangentMode=RCTM_User, time=1, value=0}
        ]}}})
    Scale RGB=&MakeVectorFromLinearColorRGB(Color=&ColorFromCurve(ColorCurve=DataInterface() {propertyValues {
        AlphaCurve {keys []}
        BlueCurve {keys [<at 0 4 RCIM_Cubic>,<at 0.7004 0.2479 RCIM_Cubic>]}
        GreenCurve {keys [<at 0 4 RCIM_Cubic>,<at 0.7004 0.7061 RCIM_Cubic>]}
        RedCurve {keys [<at 0 4 RCIM_Cubic>,<at 0.7004 1 RCIM_Cubic>]}
       }}))
    ScaleA=true
    ScaleRGB=true
   )
  ]
  rendererClasses [&NiagaraSpriteRendererProperties]
  renderers [{rendererClass=&NiagaraSpriteRendererProperties}]
 }
 FN_SpeedLines {
  simTarget=GPUComputeSim
  emitterUpdateScript [
   &EmitterState()
   &SpawnRate(
    Spawn Probability=HlslExpression() {hlslExpression="Emitter.Age > .2 ? 1 : 0"}
    SpawnRate=&FN_FrontEndFloatModifier(Gameplay=&Lerp_Float(A=0, B=50, Alpha=$System.Dive), Front End=0)
    Use Spawn Probability=true
   )
  ]
  particleSpawnScript [
   /Niagara/Modules/Spawn/Initialization/InitializeParticle(
    Color=&MakeLinearColor(
     R:NiagaraFloat=0.2292
     G:NiagaraFloat=0.1989
     B:NiagaraFloat=0.1838
     A:NiagaraFloat=&UniformRangedFloat(Minimum=0.1, Maximum=0.2)
    )
    Lifetime=&UniformRangedFloat(Minimum=0.3, Maximum=0.5)
    Module.Write Color=true
    Module.Write Lifetime=true
    Module.Write Mass=true
    Module.Write Material Random=true
    Module.Write Position=false
    Module.Write Scale=false
    Module.Write SpriteRotation=true
    Module.Write SpriteSize=true
    Position=&SimulationPosition()
    Sprite Rotation=&UniformRangedFloat(Minimum=0, Maximum=360)
    Sprite Size=&UniformRangedVector2D(Minimum=<1.25 100>, Maximum=<0.8 200>)
   ) {enabled=false}
   /Niagara/Modules/Spawn/Initialization/V2/InitializeParticle(
    Lifetime Max=0.5
    Lifetime Min=0.3
    Lifetime Mode=::NewEnumerator1
    Sprite Size Max=<1.25 200>
    Sprite Size Min=<0.8 100>
    Sprite Size Mode=::NewEnumerator2
    UsePositionOffset=false
    Write Material Random=true
   ) {module=InitializeParticle001}
   &SkeletalMeshLocation(
    Apply To Particle Position=false
    Bone / Socket Index=4
    Bone Sampling Mode=::NewEnumerator4
    Inherit Velocity=false
    MaskSpawnGroup=false
    Sampled Orientation Coordinate Space=::Local
    Skeletal Mesh=$User.SkeletalMesh
   )
   &TorusLocation(
    Align Torus With Axis=true
    Coordinate Space:ENiagaraCoordinateSpace=::World
    Handle Radius=250
    Large Radius=200
    MaskSpawnGroup=false
    Module.Override Randomness=false
    Offset=&RotateVectorByQuaternion(Vector To Rotate=<500 0 0>, Quaternion=$Output.SkeletalMeshLocation.SampledRotation)
    Torus Axis=&RotateVectorByQuaternion(Vector To Rotate=<1 0 0>, Quaternion=$Output.SkeletalMeshLocation.SampledRotation)
    Torus Origin:NiagaraPosition=$Output.SkeletalMeshLocation.SampledPosition
   ) {enabled=false}
   &SphereLocation(
    MaskSpawnGroup=false
    Non Uniform Scale=&RotateVectorByQuaternion(Vector To Rotate=<1 1 1>, Quaternion=$Output.SkeletalMeshLocation.SampledRotation)
    Offset=&Add_Vector(
     A=&RotateVectorByQuaternion(Vector To Rotate=<150 0 0>, Quaternion=$Output.SkeletalMeshLocation.SampledRotation)
     B=&TransformVector(Vector=<0 0 -200>, Source Space=::World, Destination Space=::World)
    )
    Override Seed=false
    Sphere Orientation Axis:Vector3f=&RotateVectorByQuaternion(Vector To Rotate=<100 0 0>, Quaternion=$Output.SkeletalMeshLocation.SampledRotation)
    Sphere Origin:NiagaraPosition=$Output.SkeletalMeshLocation.SampledPosition
    Surface Only=false
    UseNonUniformScale=true
    UseOffset=true
    UseSphereOrientation=true
   )
   &InheritVelocity(Module.Limit Speed=true)
  ]
  particleUpdateScript [
   &ParticleState()
   &Color(Scale Alpha=&FloatFromCurve(
     FloatCurve=DataInterface() {propertyValues {Curve {keys [<at 0 0 RCIM_Cubic>,<at 0.1855 0.9473 RCIM_Cubic>,<at 1 0 RCIM_Cubic>]}}}
     Scale Curve=0.1
    )) {enabled=false}
   &SolveForcesAndVelocity(Clamp Velocity=false, Limit Acceleration=false)
   &ScaleSpriteSize(Scale Factor=&Vector2DFromCurve(
     Vector2Curve=DataInterface() {propertyValues {
       XCurve {keys [<at -0.0069 0.5 RCIM_Cubic>,<at 0.4918 1 RCIM_Cubic>,<at 1 0.5 RCIM_Cubic>]}
       YCurve {keys [<at 0 0.0245 RCIM_Cubic>, {arriveTangent=-0.0245, interpMode=RCIM_Cubic, leaveTangent=-0.0245, time=0.4975, value=1}, <at 1 0 RCIM_Cubic>]}
      }}
     ScaleCurve=<5 3>
    ))
   &ScaleColor(
    Scale Alpha=&FloatFromCurve(
     FloatCurve=DataInterface() {propertyValues {Curve {keys [{interpMode=RCIM_Cubic, tangentMode=RCTM_User, time=0.5, value=0}, {interpMode=RCIM_Cubic, tangentMode=RCTM_User, time=0.9, value=1}]}}}
     CurveIndex=$System.Dive
     Scale Curve=0.06
    )
    ScaleA=true
    ScaleRGB=true
   )
  ]
  rendererClasses [&NiagaraSpriteRendererProperties]
  renderers [{rendererClass=&NiagaraSpriteRendererProperties}]
 }
 MoneyBags {
  emitterUpdateScript [
   &EmitterState()
   &SpawnRate(
    Spawn Probability=0.68
    SpawnRate=&FN_FrontEndFloatModifier(Gameplay=&Lerp_Float(A=2, B=3, Alpha=$System.Dive), Front End=1.5)
    Use Spawn Probability=true
   )
  ]
  particleSpawnScript [
   &InitializeParticle(
    Color=<rgba 100 45.55 1 1>
    Lifetime=&FN_FrontEndFloatModifier(Gameplay=3, Front End=3.45)
    Mesh Scale Mode=::NewEnumerator3
    Sprite Size Mode=::NewEnumerator3
    Uniform Sprite Size=45
    UsePositionOffset=false
    Write Material Random=true
   )
   &AddVelocity(Velocity=&FN_FrontEndVector3Modifier(
     Gameplay=&Lerp_Vector(A=<0 0 200>, B=<0 0 300>, Alpha=$System.Dive)
     Front End=&RandomRangeVector(Minimum=<0 0 115>, Maximum=<0 0 184>)
    ))
   &ShapeLocation(
    Disc Coverage=0.15
    MaskSpawnGroup=false
    Non Uniform Scale=<1 1 0.5>
    Offset=&Lerp_Vector(A=<0 0 -50>, B=<0 0 -100>, Alpha=$System.Dive)
    Offset Mode=::NewEnumerator0
    Override Seed=false
    Ring Radius=135
    Rotation Angle=0.65
    Rotation Axis=<0 0 1>
    Rotation Mode=::NewEnumerator1
    Shape Primitive=::NewEnumerator4
    U Distribution=0.2
    UseOwnerScale=true
   )
   &InitialMeshOrientation(
    Mesh Orientation Mode=::NewEnumerator2
    Rotation=&RandomRangeVector(Minimum=<-0.1 -0.1 -1>, Maximum=<0.1 0.1 1>)
    Use Rotation Vector=true
   )
   /SetVariables(Particles.ScaleFireworks:Vector3f=&VectorFromFloat(Value=10)) {enabled=false, isSetParameters=true}
   /SetVariables(Particles.MeshFireworksOrientation:Quat4f=&OrientMeshToVector(
     Look At Direction=&RandomRangeVector(Minimum=<-1 -1 -1>, Maximum=<1 1 1>)
     Reference Axis Vector=<1 0 0>
    )) {isSetParameters=true}
   /SetVariables(Particles.CameraOffsetGlow=1) {enabled=false, isSetParameters=true}
   /SetVariables(Particles.MeshIndexFireworks:NiagaraInt32=&RandomRangeInt(Minimum=0, Maximum=1)) {enabled=false, isSetParameters=true}
   &DynamicMaterialParameters(
    Index 1 Param 1=&SelectFloatFromArray(
     Array Sampling Mode=::NewEnumerator0
     Direct Array Index=?
     Float Selection Array=DataInterface() {propertyValues {FloatData [-1,1]}}
     Random Seed=0
    )
    Index 2 Param 1=10
    Module.Index1Param0Write=true
    Module.Index1Param1Write=false
    Module.Index1Param2Write=false
    Module.Index1Param3Write=false
    Module.Index2Param0Write=true
    Module.Index2Param1Write=true
    Module.Index2Param2Write=true
    Module.Index2Param3Write=true
    Write Parameter Index 0=false
    Write Parameter Index 1=true
    Write Parameter Index 2=true
   )
  ]
  particleUpdateScript [
   /SetVariables(Particles.PositionGlow:NiagaraPosition=&AddVectorToPosition(
     Position=$Particles.Position
     Vector=&RotateVectorByQuaternion(Vector To Rotate=<0 0 14>, Quaternion=$Particles.MeshOrientation)
    )) {isSetParameters=true}
   /SetVariables(Particles.ColorGlow:LinearColor=&MakeLinearColorFromVectorAndFloat(
     "Vector (RGB)"=&MakeVectorFromLinearColorRGB(Color=<rgba 2.5 1.1542 0.375 1>)
     "Float (Alpha)"=&FloatFromCurve(
      FloatCurve=DataInterface() {propertyValues {Curve {keys [
          {interpMode=RCIM_Cubic, tangentMode=RCTM_User, time=0.07, value=0}
          {interpMode=RCIM_Cubic, tangentMode=RCTM_User, time=0.18, value=1}
          {interpMode=RCIM_Cubic, tangentMode=RCTM_User, time=0.7, value=1}
          {interpMode=RCIM_Cubic, tangentMode=RCTM_User, time=0.8, value=0}
         ]}}}
      Scale Curve=&FN_Select_Float_from_Time_Of_Day(Morning=0.35, Day=0.3, Evening=0.35, Night=0.45)
     )
    )) {isSetParameters=true}
   /SetVariables(Particles.ColorGlowSoft:LinearColor=&MakeLinearColorFromVectorAndFloat(
     "Vector (RGB)"=&MakeVectorFromLinearColorRGB(Color=<rgba 2.5 1.1542 0.375 1>)
     "Float (Alpha)"=&FloatFromCurve(
      FloatCurve=DataInterface() {propertyValues {Curve {keys [
          {interpMode=RCIM_Cubic, tangentMode=RCTM_User, time=0.07, value=0}
          {interpMode=RCIM_Cubic, tangentMode=RCTM_User, time=0.2, value=0.4}
          {interpMode=RCIM_Cubic, tangentMode=RCTM_User, time=0.7, value=0.4}
          {interpMode=RCIM_Cubic, tangentMode=RCTM_User, time=0.8, value=0}
         ]}}}
      Scale Curve=&FN_Select_Float_from_Time_Of_Day(Morning=0.35, Day=0.3, Evening=0.35, Night=0.45)
     )
    )) {isSetParameters=true}
   &ParticleState()
   /SetVariables(Particles.ScaleGlowPlane:Vector3f=&Multiply_VectorByFloat(
     Vector=<1.25 1.4 1>
     Float=&FloatFromCurve(FloatCurve=DataInterface() {propertyValues {Curve {keys [{interpMode=RCIM_Cubic, tangentMode=RCTM_User, time=0, value=0}, {interpMode=RCIM_Cubic, tangentMode=RCTM_User, time=0.2, value=1}]}}})
    )) {isSetParameters=true}
   &ScaleSpriteSize(Uniform Curve Sprite Scale=DataInterface() {propertyValues {Curve {keys [{interpMode=RCIM_Cubic, tangentMode=RCTM_User, time=0, value=0}, {interpMode=RCIM_Cubic, tangentMode=RCTM_User, time=0.2, value=1}]}}})
   /SetVariables(Particles.SpriteSizeGlowSmall:Vector2f=&Multiply_Vector2D_ByFloat(Float=0.7, Vector2D=$Particles.SpriteSize)) {enabled=false, isSetParameters=true}
   &UpdateMeshOrientation(
    Facing Direction=&ReturnCameraVector(Camera Vectors=::NewEnumerator1, Coordinate Space=::Simulation)
    Facing Interpolation=false
    Orientation Method=::NewEnumerator1
    Override Quat=false
    TransformFacing=false
    TransformUp=false
   ) {enabled=false, module=UpdateMeshOrientation001}
   &UpdateMeshOrientation(
    Rotation Rate=&FloatFromCurve(FloatCurve=DataInterface() {propertyValues {Curve {keys [<at 0.7 1>,<at 0.75 0>]}}}, Scale Curve=0.1)
    Rotation Vector=<0 0 1>
   )
   &ScaleMeshSize(Scale Factor=&VectorFromFloat(Value=&FloatFromCurve(FloatCurve=DataInterface() {propertyValues {Curve {keys [
          {interpMode=RCIM_Cubic, tangentMode=RCTM_User, time=0, value=0}
          {interpMode=RCIM_Cubic, tangentMode=RCTM_User, time=0.2, value=1}
          {interpMode=RCIM_Cubic, tangentMode=RCTM_User, time=0.65, value=1}
          {interpMode=RCIM_Cubic, tangentMode=RCTM_User, time=0.8, value=0}
         ]}}})))
   /SetVariables(Particles.VisibilityTag=&MakeCustomIntFromBool(
     Bool=&SetBoolByFloatComparison(A=$Particles.NormalizedAge, B=0.7, Comparison Type=::NewEnumerator6)
     True INT=0
     False INT=1
    )) {isSetParameters=true}
   &ScaleVelocity(Velocity Scale=&VectorFromFloat(Value=&FloatFromCurve(FloatCurve=DataInterface() {propertyValues {Curve {keys [
          {interpMode=RCIM_Cubic, tangentMode=RCTM_User, time=0.7, value=1}
          {arriveTangent=-6.9822, interpMode=RCIM_Cubic, leaveTangent=-6.9822, tangentMode=RCTM_User, time=1, value=0.1}
         ]}}})))
   &SolveForcesAndVelocity(Clamp Velocity=false, Limit Acceleration=false)
   /SetVariables(Particles.ScaleFireworks:Vector3f=&VectorFromFloat(Value=&FloatFromCurve(
      FloatCurve=DataInterface() {propertyValues {Curve {keys [
          {arriveTangent=8.5541, interpMode=RCIM_Cubic, leaveTangent=8.5541, tangentMode=RCTM_User, time=0.7, value=0.1}
          {arriveTangent=0.0489, interpMode=RCIM_Cubic, leaveTangent=0.0489, tangentMode=RCTM_User, time=1, value=1}
         ]}}}
      Scale Curve=10
     ))) {isSetParameters=true}
   &DynamicMaterialParameters(
    Index 0 Param 1=&FloatFromCurve(
     FloatCurve=DataInterface() {propertyValues {Curve {keys [
         {interpMode=RCIM_Cubic, tangentMode=RCTM_User, time=0.6, value=3}
         {arriveTangent=-1.08, interpMode=RCIM_Cubic, leaveTangent=-1.08, tangentMode=RCTM_User, time=1, value=1}
        ]}}}
     Scale Curve=&RandomRangeFloat(Minimum=3, Maximum=4)
    )
    Index 0 Param 2=&RandomRangeFloat(Minimum=3, Maximum=4)
    Index 0 Param 3=&RandomRangeFloat(Minimum=0.5, Maximum=1)
    Index 0 Param 4=&RandomRangeFloat(Minimum=0.4, Maximum=0.7)
    Module.Index0Param0Write=true
    Module.Index0Param1Write=true
    Module.Index0Param2Write=true
    Module.Index0Param3Write=true
   ) {module=DynamicMaterialParameters001}
   &ScaleColor(
    Scale Alpha=&FloatFromCurve(FloatCurve=DataInterface() {propertyValues {Curve {keys [{interpMode=RCIM_Cubic, tangentMode=RCTM_User, time=0.7, value=1}, {interpMode=RCIM_Cubic, tangentMode=RCTM_User, time=1, value=0}]}}})
    Scale RGB=&MakeVectorFromLinearColorRGB(Color=&ColorFromCurve(ColorCurve=DataInterface() {propertyValues {
        AlphaCurve {keys []}
        BlueCurve {keys [<at 0.7452 45 RCIM_Cubic>,<at 0.9974 1 RCIM_Cubic>]}
        GreenCurve {keys [<at 0.7452 45 RCIM_Cubic>,<at 0.9974 1 RCIM_Cubic>]}
        RedCurve {keys [<at 0.7452 45 RCIM_Cubic>,<at 0.9974 1 RCIM_Cubic>]}
       }}))
    ScaleA=true
    ScaleRGB=true
   )
   /SetVariables(Particles.ScaleFireworksAlt:Vector3f=&VectorFromFloat(Value=&FloatFromCurve(
      FloatCurve=DataInterface() {propertyValues {Curve {keys [
          {arriveTangent=2.8764, interpMode=RCIM_Cubic, leaveTangent=2.8764, tangentMode=RCTM_User, time=0.7, value=0.5}
          {arriveTangent=0.0489, interpMode=RCIM_Cubic, leaveTangent=0.0489, tangentMode=RCTM_User, time=1, value=1}
         ]}}}
      Scale Curve=10
     ))) {enabled=false, isSetParameters=true}
   /SetVariables(Particles.ColorFWAlt:LinearColor=&MakeLinearColorFromVectorAndFloat(
     "Vector (RGB)"=&MakeVectorFromLinearColorRGB(Color=&ColorFromCurve(ColorCurve=DataInterface() {propertyValues {
         AlphaCurve {keys []}
         BlueCurve {keys [<at 0.6845 18.2839 RCIM_Cubic>,<at 0.8302 0.2379 RCIM_Cubic>]}
         GreenCurve {keys [<at 0.6845 41.5036 RCIM_Cubic>,<at 0.8302 0.5129 RCIM_Cubic>]}
         RedCurve {keys [<at 0.6845 100 RCIM_Cubic>,<at 0.8302 1 RCIM_Cubic>]}
        }}))
     "Float (Alpha)"=&FloatFromCurve(FloatCurve=DataInterface() {propertyValues {Curve {keys [{interpMode=RCIM_Cubic, tangentMode=RCTM_User, time=0.2, value=1}, {interpMode=RCIM_Cubic, tangentMode=RCTM_User, time=0.9, value=0}]}}})
    )) {enabled=false, isSetParameters=true}
   &GenerateLocationEvent(
    Event Probability:NiagaraFloat=&FloatFromCurve(FloatCurve=DataInterface() {propertyValues {Curve {keys [
         {arriveTangent=53.8521, interpMode=RCIM_Cubic, leaveTangent=53.8521, tangentMode=RCTM_User, time=0.68, value=0}
         {interpMode=RCIM_Cubic, tangentMode=RCTM_User, time=0.7, value=1}
         {arriveTangent=-82.9852, interpMode=RCIM_Cubic, leaveTangent=-82.9852, tangentMode=RCTM_User, time=0.72, value=0}
        ]}}})
    Event Probability Evaluation Type:ENiagaraRandomnessEvaluation=::NewEnumerator0
    Event Send Rate=40
    Use Event Probability=true
    UseEventDelay=false
   )
   /SetVariables(Particles.ScaleGlowRays:Vector3f=&Multiply_Vector(
     A=$Particles.Scale
     B=&MakeVector(
      X=1
      Y=1
      Z=&FloatFromCurve(FloatCurve=DataInterface() {propertyValues {Curve {keys [
           {interpMode=RCIM_Cubic, tangentMode=RCTM_User, time=0, value=1}
           {interpMode=RCIM_Cubic, tangentMode=RCTM_User, time=0.1, value=0.875}
           {interpMode=RCIM_Cubic, tangentMode=RCTM_User, time=0.2, value=1.175}
           {interpMode=RCIM_Cubic, tangentMode=RCTM_User, time=0.3, value=0.875}
           {interpMode=RCIM_Cubic, tangentMode=RCTM_User, time=0.4, value=1.05}
           {interpMode=RCIM_Cubic, tangentMode=RCTM_User, time=0.5, value=1.175}
           {interpMode=RCIM_Cubic, tangentMode=RCTM_User, time=0.6, value=0.875}
           {interpMode=RCIM_Cubic, tangentMode=RCTM_User, time=0.7, value=1.1}
           {interpMode=RCIM_Cubic, tangentMode=RCTM_User, time=0.8, value=0.725}
           {interpMode=RCIM_Cubic, tangentMode=RCTM_User, time=0.9, value=0.85}
           {interpMode=RCIM_Cubic, tangentMode=RCTM_User, time=1, value=1}
          ]}}})
     )
    )) {enabled=false, isSetParameters=true}
   &AlignSpriteToMeshOrientation(Mesh Orientation Relative Sprite Facing Vector:Vector3f=<0 0 1>, SetAlignment=true, SetFacing=true) {enabled=false}
  ]
  rendererClasses [&NiagaraMeshRendererProperties, &NiagaraSpriteRendererProperties]
  renderers [
   {rendererClass=&NiagaraMeshRendererProperties, rendererIndex=0}
   {rendererClass=&NiagaraMeshRendererProperties, rendererIndex=1}
   {rendererClass=&NiagaraSpriteRendererProperties, rendererIndex=2}
   {rendererClass=&NiagaraMeshRendererProperties, rendererIndex=3}
   {rendererClass=&NiagaraMeshRendererProperties, rendererIndex=4}
  ]
 }
}
summary {
 systemName=NS_Contrails_MusketSlinger
 emitters [
  {emitterName=FN_SpeedLines, simTarget=GPUComputeSim, rendererClasses [&NiagaraSpriteRendererProperties]}
  {emitterName=Bills, rendererClasses [&NiagaraMeshRendererProperties]}
  {emitterName=MoneyBags, rendererClasses [&NiagaraMeshRendererProperties, &NiagaraSpriteRendererProperties]}
  {emitterName=Coins, rendererClasses [&NiagaraMeshRendererProperties]}
  {emitterName=Bills_Stacks, rendererClasses [&NiagaraMeshRendererProperties]}
  {emitterName=Confetti, rendererClasses [&NiagaraSpriteRendererProperties]}
  {emitterName=CoinsExplosion, rendererClasses [&NiagaraMeshRendererProperties]}
  {emitterName=BillsExplosion, rendererClasses [&NiagaraMeshRendererProperties]}
 ]
 userVariables [
  User.bIsFrontEnd {type=NiagaraBool}
  User.bIsFrontEndPreview {type=NiagaraBool}
  User.ParaGlideLeanAlpha {type=NiagaraFloat, defaultValue=0, description=""}
  User.RibbonIDOffset {type=NiagaraInt32}
  User.SkeletalMesh {
   type=NiagaraDataInterfaceSkeletalMesh
   defaultValue {
    struct=NiagaraExt_VariableValue_DataInterface
    value {dataInterface=&self:NiagaraDataInterfaceSkeletalMesh_0}
   }
  }
 ]
}
